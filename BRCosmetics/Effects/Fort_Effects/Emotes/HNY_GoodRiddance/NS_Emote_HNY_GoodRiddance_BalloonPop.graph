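# Niagara one-shot burst for the balloon pop in the HNY Good Riddance emote, with no user parameters.
# Balloon_Confetti throws gold SubUV-animated sprites outward, then drag, curl noise and ramped downward acceleration make them flutter down.
# Pop_Embers launches short-lived, shrinking HDR sparks along sphere normals under light downward pull.
# SnapRing (mesh) and SnapRing001 (sprite) add brief expanding, fading shock rings, animated through a dynamic material parameter curve.
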NiagaraSystem /BRCosmetics/Effects/Fort_Effects/Emotes/HNY_GoodRiddance/NS_Emote_HNY_GoodRiddance_BalloonPop
dependencies {
 usedDataInterfaces [&NiagaraDataInterfaceCurve, &NiagaraDataInterfaceColorCurve]
 usedDynamicInputs []
 usedModules [&SystemState]
}
emitters {
 Balloon_Confetti {
  emitterUpdateScript [
   &EmitterState()
   &SpawnBurst_Instantaneous(Spawn Count=&RandomRangeInt(Minimum=10, Maximum=16), Use Spawn Probability=false)
  ]
  particleSpawnScript [
   /Niagara/Modules/Spawn/Initialization/InitializeParticle(
    Color=<rgba 0.8796 0.5581 0.0409 1>
    Lifetime=&RandomRangeFloat(Minimum=0.5, Maximum=1)
    Module.Write Color=true
    Module.Write Lifetime=true
    Module.Write Mass=true
    Module.Write Material Random=true
    Module.Write Position=false
    Module.Write Scale=false
    Module.Write SpriteRotation=true
    Module.Write SpriteSize=true
    Position=&SimulationPosition()
    Sprite Rotation=&RandomRangeFloat(Minimum=0, Maximum=360)
    Sprite Size=&Vector2DFromFloat(Value=&RandomRangeFloat(Minimum=15, Maximum=45))
   )
   /Niagara/Modules/Spawn/Location/SphereLocation(MaskSpawnGroup=false, Module.Override Randomness=false, Sphere Radius=20, Surface Only=true)
   &AddVelocityFromPoint(
    Module.Use Falloff Distance=false
    Origin Offset Coordinate Space=::Simulation
    Velocity Strength=&RandomRangeFloat(Minimum=65, Maximum=300)
   )
   &AddVelocity(Velocity=&RandomRangeVector(Minimum=<0 0 10>, Maximum=<1 1 50>))
  ]
  particleUpdateScript [
   &ParticleState()
   &Drag(Drag=5, Use Linear Drag=true, Use Rotational Drag=false)
   &AccelerationForce(Acceleration=&Multiply_VectorByFloat(
     Vector=&RandomRangeVector(Minimum=<0 0 -50>, Maximum=<0 0 -250>)
     Float=&FloatFromCurve(FloatCurve=DataInterface() {propertyValues {Curve {keys [{arriveTangent=10.9029, interpMode=RCIM_Cubic, leaveTangent=10.9029, tangentMode=RCTM_User, time=0, value=0}, <at 0.25 1>]}}})
    ))
   &CurlNoiseForce(Noise Frequency=5, Noise Strength=100, Pan Noise=false)
   &SolveForcesAndVelocity(Clamp Velocity=false, Limit Acceleration=false)
   &ScaleSpriteSize(Scale Factor=&Vector2DFromFloat(Value=&FloatFromCurve(FloatCurve=DataInterface() {propertyValues {Curve {keys [
          {arriveTangent=2.3096, interpMode=RCIM_Cubic, leaveTangent=2.3096, tangentMode=RCTM_User, time=0, value=0}
          {arriveTangent=-0.0952, interpMode=RCIM_Cubic, leaveTangent=-0.0952, tangentMode=RCTM_User, time=0.431, value=0.3973}
          <at 0.9969 0.0123 RCIM_Cubic>
         ]}}})))
   &SubUVAnimation(End Frame=36, Sub UV Play Rate=0.8, SubUV Animation Mode:ENiagara_SubUVLookupMode=::NewEnumerator3)
  ]
  rendererClasses [&NiagaraSpriteRendererProperties]
  renderers [{rendererClass=&NiagaraSpriteRendererProperties}]
 }
 Pop_Embers {
  emitterUpdateScript [
   &EmitterState()
   &SpawnBurst_Instantaneous(Spawn Count=&RandomRangeInt(Minimum=25, Maximum=50), Use Spawn Probability=false)
   &SpawnRate(SpawnRate=50000, Use Spawn Probability=false) {enabled=false}
  ]
  particleSpawnScript [
   &InitializeParticle(
    Color=<rgba 15 10 1 1>
    Lifetime=&UniformRangedFloat(Minimum=0.125, Maximum=0.375)
    Module.Write Color=true
    Module.Write Lifetime=true
    Module.Write Mass=false
    Module.Write Material Random=true
    Module.Write Position=false
    Module.Write Scale=false
    Module.Write SpriteRotation=true
    Module.Write SpriteSize=true
    Position=&SimulationPosition()
    Sprite Size=&RandomRangeVector2D(Minimum=<1 2>, Maximum=<2 2>)
   )
   &SphereLocation(MaskSpawnGroup=false, Module.Override Randomness=false, Non Uniform Scale=<1 1 0.5>, Sphere Radius=30, Surface Only=false)
   &AddVelocity(Velocity=&Multiply_VectorByFloat(Vector=$Output.SphereLocation.SphereNormal, Float=&RandomRangeFloat(Minimum=150, Maximum=200)))
  ]
  particleUpdateScript [
   &ParticleState()
   &AccelerationForce(Acceleration=<0 0 -100>)
   &SolveForcesAndVelocity(Clamp Velocity=false, Limit Acceleration=false)
   &ScaleColor(
    Scale Alpha=&FloatFromCurve(FloatCurve=DataInterface() {propertyValues {Curve {keys [{arriveTangent=11.9433, interpMode=RCIM_Cubic, time=0, value=0}, <at 1 1 RCIM_Cubic>]}}})
    Scale RGB=&MakeVectorFromLinearColorRGB(Color=&ColorFromCurve(ColorCurve=DataInterface() {propertyValues {
        AlphaCurve {keys [<at 0 1 RCIM_Cubic>]}
        BlueCurve {keys [<at 0 1 RCIM_Cubic>,<at 0.2 0 RCIM_Cubic>]}
        GreenCurve {keys [<at 0 1 RCIM_Cubic>,<at 0.2 0.5 RCIM_Cubic>]}
        RedCurve {keys [<at 0 1 RCIM_Cubic>,<at 0.2 0.5 RCIM_Cubic>]}
       }}))
    ScaleA=true
    ScaleRGB=true
   )
   &ScaleSpriteSize(Scale Factor=&Lerp_Vector2(A=<1 1>, B=<0 0.5>, Alpha=$Particles.NormalizedAge))
  ]
  rendererClasses [&NiagaraSpriteRendererProperties]
  renderers [{rendererClass=&NiagaraSpriteRendererProperties}]
 }
 SnapRing {
  emitterUpdateScript [
   &EmitterState()
   &SpawnBurst_Instantaneous(Spawn Count=1, Use Spawn Probability=false)
  ]
  particleSpawnScript [&InitializeParticle(
    Color=<rgba 2 1.5323 1 1>
    Lifetime=0.15
    Mesh Scale=&VectorFromFloat(Value=1.25)
    Module.Write Color=true
    Module.Write Lifetime=true
    Module.Write Mass=true
    Module.Write Material Random=true
    Module.Write Position=false
    Module.Write Scale=true
    Module.Write SpriteRotation=false
    Module.Write SpriteSize=true
    Position=&SimulationPosition()
    Sprite Size=&Vector2DFromFloat(Value=400)
   )]
  particleUpdateScript [
   &ParticleState()
   &DynamicMaterialParameters(
    Index 0 Param 1=&FloatFromCurve(FloatCurve=DataInterface() {propertyValues {Curve {keys [<at 0 0 RCIM_Cubic>,<at 1 1 RCIM_Cubic>]}}})
    Index 0 Param 2=50
    Index 0 Param 3=1
    Index 0 Param 4=1
    Module.Param0WriteEnabled=true
    Module.Param1WriteEnabled=true
    Module.Param2WriteEnabled=true
    Module.Param3WriteEnabled=true
   )
   &ScaleColor(
    Scale Alpha=&FloatFromCurve(FloatCurve=DataInterface() {propertyValues {Curve {keys [<at 0 1 RCIM_Cubic>,<at 1 0>]}}})
    ScaleA=true
    ScaleRGB=true
   )
   &ScaleMeshSize(Scale Factor=&VectorFromFloat(Value=&FloatFromCurve(FloatCurve=DataInterface() {propertyValues {Curve {keys [<at 0 0>,<at 0.25 1>]}}})))
  ]
  rendererClasses [&NiagaraMeshRendererProperties]
  renderers [{rendererClass=&NiagaraMeshRendererProperties, rendererIndex=0}]
 }
 SnapRing001 {
  emitterUpdateScript [
   &EmitterState()
   &SpawnBurst_Instantaneous(Spawn Count=1, Use Spawn Probability=false)
  ]
  particleSpawnScript [&InitializeParticle(
    Color=<rgba 3 2.5089 1.95 2>
    Lifetime=0.25
    Mesh Scale=&VectorFromFloat(Value=1.25)
    Module.Write Color=true
    Module.Write Lifetime=true
    Module.Write Mass=true
    Module.Write Material Random=true
    Module.Write Position=false
    Module.Write Scale=true
    Module.Write SpriteRotation=false
    Module.Write SpriteSize=true
    Position=&SimulationPosition()
    Sprite Size=&Vector2DFromFloat(Value=400)
   )]
  particleUpdateScript [
   &ParticleState()
   &DynamicMaterialParameters(
    Index 0 Param 1=&FloatFromCurve(FloatCurve=DataInterface() {propertyValues {Curve {keys [<at 0 0 RCIM_Cubic>,<at 1 1 RCIM_Cubic>]}}})
    Index 0 Param 2=50
    Index 0 Param 3=1
    Index 0 Param 4=1
    Module.Param0WriteEnabled=true
    Module.Param1WriteEnabled=true
    Module.Param2WriteEnabled=true
    Module.Param3WriteEnabled=true
   )
   &ScaleSpriteSize(Scale Factor=&Vector2DFromFloat(Value=&FloatFromCurve(FloatCurve=DataInterface() {propertyValues {Curve {keys [<at 0 0>,<at 1 1>]}}})))
   &ScaleColor(
    Scale Alpha=&FloatFromCurve(FloatCurve=DataInterface() {propertyValues {Curve {keys [<at 0.2 1 RCIM_Cubic>,<at 1 0>]}}})
    ScaleA=true
    ScaleRGB=true
   )
  ]
  rendererClasses [&NiagaraSpriteRendererProperties]
  renderers [{rendererClass=&NiagaraSpriteRendererProperties}]
 }
}
summary {
 systemName=NS_Emote_HNY_GoodRiddance_BalloonPop
 emitters [
  {emitterName=SnapRing, rendererClasses [&NiagaraMeshRendererProperties]}
  {emitterName=Balloon_Confetti, rendererClasses [&NiagaraSpriteRendererProperties]}
  {emitterName=Pop_Embers, rendererClasses [&NiagaraSpriteRendererProperties]}
  {emitterName=SnapRing001, rendererClasses [&NiagaraSpriteRendererProperties]}
 ]
 userVariables []
}
