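# Niagara gold sparkle effect for the HNY_GoodRiddance emote, built from two sprite emitters.
# ParticleDecal bursts one large sprite after a short delay, animating dynamic material parameters and alpha with curves over its life.
# Sparkles_UpwardMotion continuously spawns small, short-lived sprites on a stretched sphere surface, drifting radially under drag.
# Sparkles twinkle through a multi-peak alpha curve, peak in over-bright color mid-life, and shrink out.
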NiagaraSystem /BRCosmetics/Effects/Fort_Effects/Emotes/HNY_GoodRiddance/NS_Emote_HNY_GoodRiddance_Sparkle
dependencies {
 usedDataInterfaces [&NiagaraDataInterfaceColorCurve, &NiagaraDataInterfaceCurve]
 usedDynamicInputs []
 usedModules [&SystemState]
}
emitters {
 ParticleDecal {
  emitterUpdateScript [
   &EmitterState(Life Cycle Mode=::NewEnumerator1, Loop Behavior=::NewEnumerator1, Loop Delay=0.25, Loop Duration=4, UseLoopDelay=true)
   &SpawnBurst_Instantaneous(Spawn Count=1, Use Spawn Probability=false)
  ]
  particleSpawnScript [
   /Niagara/Modules/Spawn/Initialization/InitializeParticle(
    Color=<rgba 1 0.8968 0.2207 1>
    Lifetime=2.75
    Module.Write Color=true
    Module.Write Lifetime=true
    Module.Write Mass=true
    Module.Write Material Random=true
    Module.Write Position=true
    Module.Write Scale=false
    Module.Write SpriteRotation=false
    Module.Write SpriteSize=true
    Sprite Size=&Vector2DFromFloat(Value=70)
   )
   &CameraOffset(Camera Offset Amount=25)
  ]
  particleUpdateScript [
   &ParticleState()
   &DynamicMaterialParameters(
    Index 0 Param 1=&FloatFromCurve(FloatCurve=DataInterface() {propertyValues {Curve {keys [
         {interpMode=RCIM_Cubic, leaveTangent=8.739, tangentMode=RCTM_Break, time=0, value=-0.5}
         {arriveTangent=23.3981, interpMode=RCIM_Cubic, leaveTangent=23.3981, time=0.0514, value=0.8887}
         {arriveTangent=8.2492, interpMode=RCIM_Cubic, leaveTangent=8.2492, tangentMode=RCTM_User, time=0.1496, value=3}
         <at 1 6 RCIM_Cubic>
        ]}}})
    Index 0 Param 2=1
    Index 0 Param 3=&FloatFromCurve(FloatCurve=DataInterface() {propertyValues {Curve {keys [
         {interpMode=RCIM_Cubic, leaveTangent=2.9398, tangentMode=RCTM_Break, time=0, value=-0.25}
         {arriveTangent=0.3723, interpMode=RCIM_Cubic, tangentMode=RCTM_Break, time=1, value=1.4}
        ]}}})
    Index 0 Param 4=1
    Module.Param0WriteEnabled=true
    Module.Param1WriteEnabled=true
    Module.Param2WriteEnabled=true
    Module.Param3WriteEnabled=true
   )
   &ScaleColor(
    Scale Alpha=&FloatFromCurve(FloatCurve=DataInterface() {propertyValues {Curve {keys [<at 0.05 0 RCIM_Cubic>, {arriveTangent=1.1111, time=0.2, value=1}, <at 0.9 1 RCIM_Cubic>, <at 1 0>]}}})
    ScaleA=true
    ScaleRGB=true
   )
  ]
  rendererClasses [&NiagaraSpriteRendererProperties]
  renderers [{rendererClass=&NiagaraSpriteRendererProperties}]
 }
 Sparkles_UpwardMotion {
  emitterUpdateScript [
   &EmitterState()
   &SpawnRate(SpawnRate=10, Use Spawn Probability=false)
  ]
  particleSpawnScript [
   &InitializeParticle(
    Color=<rgba 1 0.925 0.25 1>
    Lifetime=&UniformRangedFloat(Minimum=0.5, Maximum=0.6)
    Mesh Scale=<1.5 1.5 1.5>
    Module.Write Color=true
    Module.Write Lifetime=true
    Module.Write Mass=false
    Module.Write Material Random=true
    Module.Write Position=false
    Module.Write Scale=false
    Module.Write SpriteRotation=true
    Module.Write SpriteSize=true
    Position=&SimulationPosition()
    Sprite Rotation=&UniformRangedFloat(Minimum=0, Maximum=1)
    Sprite Size=&UniformRangedVector2D(
     Minimum=&Vector2DFromFloat(Value=8)
     Maximum=&Vector2DFromFloat(Value=16)
    )
   )
   /Niagara/Modules/Spawn/Location/SphereLocation(MaskSpawnGroup=false, Module.Override Randomness=false, Non Uniform Scale=<1 1.75 1>, Sphere Radius=20, Surface Only=true) {module=SphereLocation001}
   &AddVelocityFromPoint(
    Module.Use Falloff Distance=false
    Origin Offset Coordinate Space=::World
    Velocity Strength=&RandomRangeFloat(Minimum=-5, Maximum=10)
   )
   &CameraOffset(Camera Offset Amount=5)
  ]
  particleUpdateScript [
   &ParticleState()
   &Color(Color=&ColorFromCurve(ColorCurve=DataInterface() {propertyValues {
       AlphaCurve {keys [
         <at -0.0023 0 RCIM_Cubic>
         {arriveTangent=0.2428, interpMode=RCIM_Cubic, leaveTangent=0.2428, time=0.1, value=1}
         {arriveTangent=-0.0163, interpMode=RCIM_Cubic, leaveTangent=-0.0163, time=0.2779, value=0.0681}
         {arriveTangent=-0.0735, interpMode=RCIM_Cubic, leaveTangent=-0.0735, time=0.4489, value=0.9943}
         {arriveTangent=0.0148, interpMode=RCIM_Cubic, leaveTangent=0.0148, time=0.6472, value=0.0409}
         {arriveTangent=-0.1252, interpMode=RCIM_Cubic, leaveTangent=-0.1252, time=0.8337, value=1}
         <at 0.9977 -0.003 RCIM_Cubic>
        ]}
       BlueCurve {keys [<at -0.0023 0.6 RCIM_Cubic>,<at 0.448 2.2>,<at 0.9977 0.4 RCIM_Cubic>]}
       GreenCurve {keys [<at -0.0023 1.65 RCIM_Cubic>,<at 0.448 3.55>,<at 0.9977 1.6 RCIM_Cubic>]}
       RedCurve {keys [<at -0.0023 2 RCIM_Cubic>,<at 0.448 4>,<at 0.9977 2 RCIM_Cubic>]}
      }}))
   &Drag(Drag=5, Use Linear Drag=true, Use Rotational Drag=false)
   &SolveForcesAndVelocity(Clamp Velocity=false, Limit Acceleration=false)
   &ScaleSpriteSize(Scale Factor=&Vector2DFromFloat(Value=&FloatFromCurve(FloatCurve=DataInterface() {propertyValues {Curve {keys [<at 0.5 1>,<at 1 0>]}}})))
  ]
  rendererClasses [&NiagaraSpriteRendererProperties]
  renderers [{rendererClass=&NiagaraSpriteRendererProperties}]
 }
}
summary {
 systemName=NS_Emote_HNY_GoodRiddance_Sparkle
 emitters [
  {emitterName=Sparkles_UpwardMotion, rendererClasses [&NiagaraSpriteRendererProperties]}
  {emitterName=ParticleDecal, rendererClasses [&NiagaraSpriteRendererProperties]}
 ]
 userVariables []
}
